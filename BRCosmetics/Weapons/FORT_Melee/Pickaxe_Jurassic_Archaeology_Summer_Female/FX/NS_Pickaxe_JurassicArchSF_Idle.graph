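# Idle sparkle effect for the Jurassic Archaeology Summer Female pickaxe, scattering glints over its axe head.
# A single sprite emitter fires small random bursts in loops with randomized duration and delay.
# Spawn positions are sampled from the AxeHead static mesh, set as the default mesh, with a slight normal offset.
# Bright, randomly rotated sprites pulse color, alpha and size up and down over their lifetime, pulled toward camera, with force solving disabled.
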NiagaraSystem /BRCosmetics/Weapons/FORT_Melee/Pickaxe_Jurassic_Archaeology_Summer_Female/FX/NS_Pickaxe_JurassicArchSF_Idle
dependencies {
 usedDataInterfaces [&NiagaraDataInterfaceStaticMesh, &NiagaraDataInterfaceCurve]
 usedDynamicInputs []
 usedModules [&SystemState]
}
emitters {AxeHead_Sparkles {
  emitterUpdateScript [
   &EmitterState(
    Life Cycle Mode=::NewEnumerator1
    Loop Delay=&UniformRangedFloat(Minimum=0.5, Maximum=1)
    Loop Duration=&UniformRangedFloat(Minimum=0.5, Maximum=1)
    Recalculate Duration Each Loop=true
    UseLoopDelay=true
   )
   &SpawnBurst_Instantaneous(
    Spawn Count=&RandomRangeInt(Minimum=2, Maximum=6)
    Spawn Time=&RandomRangeFloat(Minimum=0, Maximum=1)
    Use Spawn Probability=false
   )
  ]
  particleSpawnScript [
   /Niagara/Modules/Spawn/Initialization/InitializeParticle(
    Color=<rgba 5 5 5 1>
    Lifetime=&UniformRangedFloat(Minimum=0.3, Maximum=5)
    Module.Write Color=true
    Module.Write Lifetime=true
    Module.Write Mass=true
    Module.Write Material Random=true
    Module.Write Position=true
    Module.Write Scale=false
    Module.Write SpriteRotation=true
    Module.Write SpriteSize=true
    Sprite Rotation=&UniformRangedFloat(Minimum=0, Maximum=360)
    Sprite Size=&Vector2DFromFloat(Value=&UniformRangedFloat(Minimum=10, Maximum=30))
   ) {enabled=false}
   /Niagara/Modules/Spawn/Initialization/V2/InitializeParticle(
    Color=<rgba 5 5 5 1>
    Lifetime=&UniformRangedFloat(Minimum=0.3, Maximum=5)
    Mass Mode=::NewEnumerator1
    Sprite Rotation Angle=&UniformRangedFloat(Minimum=0, Maximum=360)
    Sprite Rotation Mode=::NewEnumerator0
    Sprite Size=&Vector2DFromFloat(Value=&UniformRangedFloat(Minimum=10, Maximum=30))
    Sprite Size Mode=::NewEnumerator4
    UsePositionOffset=false
    Write Material Random=true
   ) {module=InitializeParticle001}
   /Niagara/Modules/Spawn/MeshInterface/SampleStaticMesh(
    OverrideRandomSeed=false
    StaticMesh=DataInterface() {propertyValues {
      DefaultMesh=/BRCosmetics/Weapons/FORT_Melee/Pickaxe_Jurassic_Archaeology_Summer_Female/Meshes/Jurassic_Archaeology_SF_AxeHead
      InstanceIndex=-1
      LODIndex=0
      PreviewMesh=&Jurassic_Archaeology_SF_AxeHead
      SoftSourceActor=null
      SourceMode=DefaultMeshOnly
      bAllowSamplingFromStreamingLODs=false
      bCaptureTransformsPerFrame=false
      bUsePhysicsBodyVelocity=false
      FilteredSockets []
      LODIndexUserParameter {parameter {name=None, typeDefHandle {registeredTypeIndex=108}, varData []}}
      MeshParameterBinding {parameter {name=None, typeDefHandle {registeredTypeIndex=1}, varData []}}
      SectionFilter {allowedMaterialSlots []}
     }}
   )
   /Niagara/Modules/Spawn/Location/StaticMeshLocation(Module.Enable Axis Constraint=false, Normal Offset=1)
   /SetVariables(Particles.PositionSwitch:Vector3f=&MaskVectorBySpawnGroup(
     Masked Vector=Linked() {linkedVariable {name=Particles.Position, type=NiagaraPosition}}
     Passthrough Vector=Linked() {linkedVariable {name=Output.StaticMeshLocation.Position, type=NiagaraPosition}}
    )) {isSetParameters=true}
  ]
  particleUpdateScript [
   &ScaleColor(
    Scale Alpha=&FloatFromCurve(FloatCurve=DataInterface() {propertyValues {Curve {keys [<at 0 0>,<at 0.5016 0.9767>,<at 1 0>]}}})
    Scale RGB=&VectorFromFloat(Value=&FloatFromCurve(FloatCurve=DataInterface() {propertyValues {Curve {keys [<at 0 0>,<at 0.5017 1.0042>,<at 1 0>]}}}))
    ScaleA=true
    ScaleRGB=true
   )
   &ParticleState()
   &SolveForcesAndVelocity(Clamp Velocity=false, Limit Acceleration=false) {enabled=false}
   &ScaleSpriteSize(Scale Factor=&Vector2DFromFloat(Value=&FloatFromCurve(FloatCurve=DataInterface() {propertyValues {Curve {keys [
          <at 0 0 RCIM_Cubic>
          {arriveTangent=1.9892, interpMode=RCIM_Cubic, leaveTangent=1.9892, time=0.335, value=0.2056}
          {arriveTangent=0.2473, interpMode=RCIM_Cubic, leaveTangent=0.2473, time=0.5002, value=0.995}
          {arriveTangent=-1.9908, interpMode=RCIM_Cubic, leaveTangent=-1.9908, time=0.6669, value=0.2877}
          <at 1 0>
         ]}}})))
   &CameraOffset(Camera Offset Amount=30)
  ]
  rendererClasses [&NiagaraSpriteRendererProperties]
  renderers [{rendererClass=&NiagaraSpriteRendererProperties}]
 }}
summary {
 systemName=NS_Pickaxe_JurassicArchSF_Idle
 emitters [{emitterName=AxeHead_Sparkles, rendererClasses [&NiagaraSpriteRendererProperties]}]
 userVariables []
}
